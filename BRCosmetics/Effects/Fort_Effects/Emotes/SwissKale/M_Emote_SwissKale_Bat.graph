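# Unlit, translucent, two-sided Niagara mesh particle material that draws a smoky bat for the SwissKale emote.
# Emissive tints panning, noise-warped smoke, mirrored across left and right UV halves with sheared, rotated coordinates, by a dark color and particle color.
# Opacity uses the bat texture, distorted per side by scrolling noise, modulated by mirrored noise, vertex color, particle alpha and view-angle fade.
# Dynamic parameters offset the smoke and drive a HeightLerp noise dissolve.
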Material /BRCosmetics/Effects/Fort_Effects/Emotes/SwissKale/M_Emote_SwissKale_Bat
graph {
 properties {
  bAutomaticallySetUsageInEditor=false
  bUsedWithNiagaraMeshParticles=true
  blendMode=BLEND_Translucent
  shadingModel=MSM_Unlit
  twoSided=true
  usedShadingModels=Unlit
 }
 nodes [
  Constant() {r=0.01}
  n1 = VectorParameter() {defaultValue=<rgba 0.05 0.05 0.05 1>}
  n2 = TextureCoordinate()
  n3 = ComponentMask(None=@n2) {r=true}
  n4 = ComponentMask(None=@n2) {g=true}
  n5 = AppendVector(A=@n3, B=Add(A=@n3, B=Multiply(A=@n4) {constB=3}))
  n6 = Constant2Vector() {g=0.7, r=1}
  n7 = Add(
   A=MaterialFunctionCall(UVs=Multiply(A=@n5, B=@n6), "Rotation Angle (0-1)"=Constant() {r=0.1}) {materialFunction=&CustomRotator}.Rotated Values
   B=AppendVector(A=DynamicParameter() {defaultValue=<rgba 0 1 1 1>, paramNames [Disappear,Offset,Param3,Param4]}.Offset, B=Constant())
  )
  n8 = TextureSample(UVs=Panner(Coordinate=@n7) {speedX=-0.1}) {texture=/BRCosmetics/Weapons/FORT_Melee/Pickaxe_CoatCheck/FX/T_Pickaxe_CoatCheck_Noise}
  n9 = Panner(Coordinate=Add(A=@n7, B=Multiply(A=@n8.R) {constB=0.2})) {speedX=-0.5}
  n10 = LinearInterpolate(
   A=TextureSample(UVs=@n9) {texture=&T_Inky_Smoke_Tile}.R
   B=TextureSample(UVs=@n9) {samplerType=SAMPLERTYPE_Grayscale, texture=&/T_Blur_Noise}.R
  )
  n11 = ComponentMask(None=TextureCoordinate()) {r=true}
  n12 = Saturate(None=Multiply(A=SmoothStep(Value=@n11) {constMin=0.45}) {constB=5})
  n13 = TextureCoordinate() {uTiling=-1}
  n14 = ComponentMask(None=@n13) {r=true}
  n15 = ComponentMask(None=@n13) {g=true}
  n16 = AppendVector(A=@n14, B=Add(A=@n14, B=Multiply(A=@n15) {constB=3}))
  n17 = Constant2Vector() {g=0.7, r=1}
  n18 = Add(
   A=MaterialFunctionCall(UVs=Multiply(A=@n16, B=@n17), "Rotation Angle (0-1)"=Constant() {r=0.1}) {materialFunction=&CustomRotator}.Rotated Values
   B=AppendVector(A=DynamicParameter() {defaultValue=<rgba 0 1 1 1>, paramNames [Disappear,Offset,Param3,Param4]}.Offset, B=Constant())
  )
  n19 = Panner(Coordinate=@n18) {speedX=-0.1}
  n20 = Add(
   A=Panner(Coordinate=@n18) {speedX=-0.5}
   B=Multiply(A=TextureSample(UVs=@n19) {texture=&T_Pickaxe_CoatCheck_Noise}.R) {constB=0.2}
  )
  n21 = LinearInterpolate(
   A=TextureSample(UVs=@n20) {texture=&T_Inky_Smoke_Tile}.R
   B=TextureSample(UVs=@n20) {samplerType=SAMPLERTYPE_Grayscale, texture=&T_Blur_Noise}.R
  )
  n22 = OneMinus(None=ComponentMask(None=TextureCoordinate()) {r=true})
  n23 = Saturate(None=Multiply(A=SmoothStep(Value=@n22) {constMin=0.45}) {constB=5})
  n24 = LinearInterpolate(Alpha=Add(A=Multiply(A=@n10, B=@n12), B=Multiply(A=@n21, B=@n23))) {constA=0.1, constB=1.3}
  MaterialFunctionCall(Emissive=Multiply(A=Multiply(A=@n1.RGB, B=@n24), B=ParticleColor().RGB)) {materialFunction=&MF_FX_EmissiveAdjustments}
  n25 = ParticleColor()
  n26 = TextureCoordinate()
  n27 = ComponentMask(None=@n26) {r=true}
  n28 = Multiply(A=ComponentMask(None=@n26) {g=true}) {constB=3}
  n29 = Multiply(A=AppendVector(A=@n27, B=Add(A=@n27, B=@n28)), B=Constant2Vector() {g=0.7, r=1})
  n30 = Constant() {r=0.1}
  n31 = TextureSample(UVs=Panner(Coordinate=MaterialFunctionCall(UVs=@n29, "Rotation Angle (0-1)"=@n30) {materialFunction=&CustomRotator}.Rotated Values) {speedX=-1}) {samplerType=SAMPLERTYPE_Grayscale, texture=&LowResBlurredNoise}
  n32 = ComponentMask(None=TextureCoordinate()) {r=true}
  n33 = Saturate(None=Multiply(A=SmoothStep(Value=@n32) {constMin=0.4}) {constB=5})
  n34 = TextureCoordinate() {uTiling=-1}
  n35 = ComponentMask(None=@n34) {r=true}
  n36 = Multiply(A=ComponentMask(None=@n34) {g=true}) {constB=3}
  n37 = Multiply(A=AppendVector(A=@n35, B=Add(A=@n35, B=@n36)), B=Constant2Vector() {g=0.7, r=1})
  n38 = Constant() {r=0.1}
  n39 = TextureSample(UVs=Panner(Coordinate=MaterialFunctionCall(UVs=@n37, "Rotation Angle (0-1)"=@n38) {materialFunction=&CustomRotator}.Rotated Values) {speedX=-1}) {samplerType=SAMPLERTYPE_Grayscale, texture=&LowResBlurredNoise}
  n40 = OneMinus(None=ComponentMask(None=TextureCoordinate()) {r=true})
  n41 = Saturate(None=Multiply(A=SmoothStep(Value=@n40) {constMin=0.4}) {constB=5})
  n42 = LinearInterpolate(Alpha=Add(A=Multiply(A=@n31.R, B=@n33), B=Multiply(A=@n39.R, B=@n41))) {constA=0.56}
  n43 = TextureCoordinate()
  n44 = ComponentMask(None=@n43) {r=true}
  n45 = Add(A=@n44, B=Multiply(A=ComponentMask(None=@n43) {g=true}) {constB=3})
  n46 = MaterialFunctionCall(UVs=Multiply(A=AppendVector(A=@n44, B=@n45), B=Constant2Vector() {g=1, r=1}), "Rotation Angle (0-1)"=Constant() {r=0.1}) {materialFunction=&CustomRotator}
  n47 = TextureSample(UVs=NamedRerouteDeclaration(None=Panner(Coordinate=@n46.Rotated Values) {speedX=-1}) {name=DisturbPannnerR, nodeColor=<rgba 0.875 0 1 1>}) {samplerType=SAMPLERTYPE_Normal, texture=/Game/Characters/Player/Female/Medium/Bodies/F_MED_Bistro_Astronaut/FX/T_GaussianNoise_Normal}
  n48 = ComponentMask(None=TextureCoordinate()) {r=true}
  n49 = Saturate(None=Multiply(A=SmoothStep(Value=@n48) {constMin=0.4}) {constB=3})
  n50 = Multiply(
   A=Multiply(A=AppendVector(A=@n47.R, B=@n47.R), B=NamedRerouteDeclaration(None=@n49) {name=MaskL, nodeColor=<rgba 1 0.6734 0.1165 1>})
   B=ScalarParameter() {defaultValue=-0.02, parameterName=Param1}
  ) {constB=0.2}
  n51 = TextureCoordinate() {uTiling=-1}
  n52 = ComponentMask(None=@n51) {r=true}
  n53 = Add(A=@n52, B=Multiply(A=ComponentMask(None=@n51) {g=true}) {constB=2})
  n54 = MaterialFunctionCall(UVs=Multiply(A=AppendVector(A=@n52, B=@n53), B=Constant2Vector() {g=1, r=1}), "Rotation Angle (0-1)"=Constant() {r=0.1}) {materialFunction=&CustomRotator}
  n55 = TextureSample(UVs=NamedRerouteDeclaration(None=Panner(Coordinate=@n54.Rotated Values) {speedX=-1}) {name=DisturbPannerL, nodeColor=<rgba 0 1 0.5547 1>}) {samplerType=SAMPLERTYPE_Normal, texture=&T_GaussianNoise_Normal}
  n56 = OneMinus(None=ComponentMask(None=TextureCoordinate()) {r=true})
  n57 = Saturate(None=Multiply(A=SmoothStep(Value=@n56) {constMin=0.4}) {constB=3})
  n58 = Multiply(
   A=Multiply(A=AppendVector(A=@n55.R, B=@n55.R), B=NamedRerouteDeclaration(None=@n57) {name=MaskR, nodeColor=<rgba 1 0.0933 0 1>})
   B=ScalarParameter() {defaultValue=0.02, parameterName=Param2}
  ) {constB=0}
  n59 = TextureSample(UVs=Add(A=TextureCoordinate(), B=Add(A=@n50, B=@n58)) {constB=0}) {samplerType=SAMPLERTYPE_Grayscale, texture=&/T_Emote_SwissKale_Bat}
  n60 = VertexColor()
  n61 = MaterialFunctionCall("FadeSensitivity (optional)"=ScalarParameter() {defaultValue=2}) {materialFunction=&FadeBasedOnViewAngle}
  n62 = Multiply(A=Multiply(A=Multiply(A=@n59.R, B=@n42), B=Power(Base=@n60.R)), B=@n61.Result)
  n63 = MaterialFunctionCall(
   Transition Phase=DynamicParameter() {defaultValue=<rgba 0 1 1 1>, paramNames [Disappear,Offset,Param3,Param4]}.Disappear
   Height Texture=TextureSample(UVs=TextureCoordinate() {uTiling=1.5, vTiling=3}) {samplerType=SAMPLERTYPE_Grayscale, texture=&LowResBlurredNoise}.R
  ) {materialFunction=&HeightLerp}
  Multiply(A=Multiply(A=Multiply(A=@n25.A, B=@n62)) {constB=1.5}, B=Saturate(None=Multiply(A=@n63.Results) {constB=2}))
  Saturate(None=Multiply(A=@n61.Result) {constB=3})
 ]
}
