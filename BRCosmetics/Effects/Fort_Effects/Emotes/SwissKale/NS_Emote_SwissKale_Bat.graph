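# Niagara effect for the SwissKale emote that forms a bat shape from layered meshes and smoke.
# Two mesh emitters rise and grow, cancelling owner scale, with animated dynamic material parameters and alpha delayed by emitter age.
# Two SubUV sprite smoke emitters sample the bat static mesh in particle update to follow its silhouette.
# Camera-distance remaps adjust size, placement, opacity and spawning, while front-end preview and time of day drive tinting.
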NiagaraSystem /BRCosmetics/Effects/Fort_Effects/Emotes/SwissKale/NS_Emote_SwissKale_Bat
dependencies {
 usedDataInterfaces [
  &NiagaraDataInterfaceCurve
  &NiagaraDataInterfaceStaticMesh
  &NiagaraDataInterfaceSpriteRendererInfo
  &NiagaraDataInterfaceColorCurve
 ]
 usedDynamicInputs []
 usedModules [&SystemState]
}
emitters {
 BatMesh {
  emitterUpdateScript [
   &EmitterState(Life Cycle Mode=::NewEnumerator1, Loop Behavior=::NewEnumerator1, Loop Delay=0.3, UseLoopDelay=true)
   &SpawnBurst_Instantaneous(Use Spawn Probability=false, UseLoopCountLimit=false)
  ]
  particleSpawnScript [
   &InitializeParticle(
    Color=&MakeCustomLinearColorFromBool(Bool=$User.bIsFrontEndPreview, True Color=<rgba 0.65 0.8125 1.3 0.9>, False Color=<rgba 1.3 1.3 1.3 0.9>)
    Lifetime=3.5
    Mesh Scale=&Multiply_VectorByFloat(Vector=<1 1 1>, Float=1)
    Mesh Scale Mode=::NewEnumerator4
    UsePositionOffset=false
    Write Material Random=true
   )
   &InitialMeshOrientation(Mesh Orientation Mode=::NewEnumerator2, Rotation=<0 0 -0.25>, Use Rotation Vector=true)
   &CameraOffset()
  ]
  particleUpdateScript [
   /SetVariables(Particles.Position=&AddVectorToPosition(
     Position=$Engine.Emitter.SimulationPosition
     Vector=&Lerp_Vector(
      A=<0 0 0>
      B=<0 0 10>
      Alpha=&FloatFromCurve(FloatCurve=DataInterface() {propertyValues {Curve {keys [
           {arriveTangent=0.9174, interpMode=RCIM_Cubic, leaveTangent=0.9174, tangentMode=RCTM_User, time=0, value=0}
           {arriveTangent=0.4774, tangentMode=RCTM_User, time=1, value=1}
          ]}}})
     )
    )) {isSetParameters=true, module=SetVariables_#1}
   &ParticleState()
   &ScaleMeshSize(Scale Factor=&Multiply_VectorByFloat(
     Vector=&Divide_Vector(A:Vector3f=&VectorFromFloat(Value=1), B:Vector3f=$Engine.Owner.Scale)
     Float=&FloatFromCurve(
      FloatCurve=DataInterface() {propertyValues {Curve {keys [<at 0 0>,<at 0.2 2.2>,<at 1 2.9>]}}}
      Scale Curve=&Multiply_Float(
       A=0.7
       B=&FN_LOD_Distance_Remap(Start Distance=300, End Distance=400, Minimum Output=0.8, Maximum Output=1, Clamp Output=true)
      )
     )
    ))
   &DynamicMaterialParameters(
    Index 0 Param 1=&FloatFromCurve(FloatCurve=DataInterface() {propertyValues {Curve {keys [<at 0 1>,<at 0.5 0>,<at 0.65 0>,<at 0.85 1>]}}})
    Module.Index0Param0Write=true
    Module.Index0Param1Write=true
    Module.Index0Param2Write=true
    Module.Index0Param3Write=true
   )
   /SetVariables(Particles.Position=$Particles.Position) {enabled=false, isSetParameters=true, module=SetVariables_#2}
   &ScaleColor(
    Scale Alpha=&Multiply_Float(
     A=&FN_FrontEndFloatModifier(Gameplay=1.05, Front End=1)
     B=&FloatFromCurve(FloatCurve=DataInterface() {propertyValues {Curve {keys [<at 1 0>,<at 1.1 1>,<at 1.5 1>]}}}, CurveIndex=$Emitter.Age)
    )
    Scale RGB=&MakeVectorFromLinearColorRGB(Color=&FNTimeOfDayColor(Night Color=<rgba 0.7594 1.0696 2 1>, Morning Color=<rgba 1 1 1 1>, Day Color=<rgba 1 1 1 1>, Evening Color=<rgba 1 1 1 1>))
    ScaleA=true
    ScaleRGB=true
   )
  ]
  rendererClasses [&NiagaraMeshRendererProperties]
  renderers [{rendererClass=&NiagaraMeshRendererProperties, rendererIndex=0}]
 }
 BatMesh02 {
  emitterUpdateScript [
   &EmitterState(Life Cycle Mode=::NewEnumerator1, Loop Behavior=::NewEnumerator1, Loop Delay=0.3, UseLoopDelay=true)
   &SpawnBurst_Instantaneous(Use Spawn Probability=false, UseLoopCountLimit=false)
  ]
  particleSpawnScript [
   &InitializeParticle(
    Color=&MakeCustomLinearColorFromBool(Bool=$User.bIsFrontEndPreview, True Color=<rgba 1 1.25 2 0.85>, False Color=<rgba 2 2 2 0.85>)
    Lifetime=3
    Mesh Scale=&Multiply_VectorByFloat(Vector=<1 1 1>, Float=0.9)
    Mesh Scale Mode=::NewEnumerator4
    Position Offset=<0 0 -1>
    UsePositionOffset=true
    Write Material Random=true
   )
   &InitialMeshOrientation(Mesh Orientation Mode=::NewEnumerator2, Rotation=<0 0 -0.25>, Use Rotation Vector=true)
   &CameraOffset()
  ]
  particleUpdateScript [
   /SetVariables() {enabled=false, isSetParameters=true, module=SetVariables_#1}
   /SetVariables(Particles.Position=&AddVectorToPosition(
     Position=$Engine.Emitter.SimulationPosition
     Vector=&Lerp_Vector(
      A=<0 0 0>
      B=<0 0 10>
      Alpha=&FloatFromCurve(FloatCurve=DataInterface() {propertyValues {Curve {keys [{arriveTangent=1.7443, interpMode=RCIM_Cubic, leaveTangent=1.7443, tangentMode=RCTM_User, time=0, value=0}, <at 1 1>]}}})
     )
    )) {isSetParameters=true}
   &ParticleState()
   &ScaleMeshSize(Scale Factor=&Multiply_VectorByFloat(
     Vector=&Divide_Vector(A:Vector3f=&VectorFromFloat(Value=1), B:Vector3f=$Engine.Owner.Scale)
     Float=&FloatFromCurve(
      FloatCurve=DataInterface() {propertyValues {Curve {keys [<at 0 0>,<at 0.2 2.2>,<at 1 2.9>]}}}
      Scale Curve=&Multiply_Float(
       A=0.7
       B=&FN_LOD_Distance_Remap(Start Distance=300, End Distance=400, Minimum Output=0.8, Maximum Output=1, Clamp Output=true)
      )
     )
    )) {module=ScaleMeshSize001}
   &ScaleMeshSize(Scale Factor=&Multiply_VectorByFloat(
     Vector=<1 1 1>
     Float=&FloatFromCurve(FloatCurve=DataInterface() {propertyValues {Curve {keys [<at 0 0.8>,<at 1 1.5>]}}}, Scale Curve=0.8)
    )) {enabled=false}
   &DynamicMaterialParameters(
    Index 0 Param 1=&FloatFromCurve(FloatCurve=DataInterface() {propertyValues {Curve {keys [<at 0 1>,<at 0.5 0>,<at 0.65 0>,<at 0.85 1>]}}})
    Index 0 Param 2=0.4
    Module.Index0Param0Write=true
    Module.Index0Param1Write=true
    Module.Index0Param2Write=true
    Module.Index0Param3Write=true
   )
   /SetVariables(Particles.Position=$Particles.Position) {enabled=false, isSetParameters=true, module=SetVariables_#2}
   &ScaleColor(
    Scale Alpha=&Multiply_Float(
     A=0.6
     B=&FloatFromCurve(FloatCurve=DataInterface() {propertyValues {Curve {keys [<at 1 0>,<at 1.1 1>,<at 1.5 1>]}}}, CurveIndex=$Emitter.Age)
    )
    Scale RGB=&MakeVectorFromLinearColorRGB(Color=&FNTimeOfDayColor(Night Color=<rgba 0.7594 1.0696 2 1>, Morning Color=<rgba 1 1 1 1>, Day Color=<rgba 1 1 1 1>, Evening Color=<rgba 1 1 1 1>))
    ScaleA=true
    ScaleRGB=true
   )
  ]
  rendererClasses [&NiagaraMeshRendererProperties]
  renderers [{rendererClass=&NiagaraMeshRendererProperties, rendererIndex=0}]
 }
 MeshSmokeBlur {
  emitterUpdateScript [
   &EmitterState(Life Cycle Mode=::NewEnumerator1, Loop Behavior=::NewEnumerator1, Loop Duration=1.7, UseLoopDelay=false)
   &SpawnBurst_Instantaneous(Spawn Count=30, Use Spawn Probability=false, UseLoopCountLimit=false)
   &SpawnRate(SpawnRate=30, Use Spawn Probability=true)
  ]
  particleSpawnScript [&InitializeParticle(
    Color=<rgba 0.1823 0.1823 0.1823 0.2>
    Lifetime Max=2.5
    Lifetime Min=2.5
    Lifetime Mode=::NewEnumerator1
    Mesh Scale Mode=::NewEnumerator4
    Position Offset=<0 0 -7>
    Sprite Rotation Mode=::NewEnumerator2
    Sprite Size Mode=::NewEnumerator1
    Uniform Sprite Size Max=150
    Uniform Sprite Size Min=100
    UsePositionOffset=true
    Write Material Random=true
   )]
  particleUpdateScript [
   /Niagara/Modules/Spawn/Location/StaticMeshLocation(
    MaskSpawnGroup=false
    Offset Position Along Sampled Normal=0.5
    OffsetAlongNormal=true
    Orientation Sampling:ENiagara_AttributeSamplingApplyOutput=::NewEnumerator1
    OverrideRandomSeed=false
    SamplePosOffset=false
    Static Mesh=DataInterface() {propertyValues {
      DefaultMesh=&/SM_Emote_SwissKale_Bat
      InstanceIndex=-1
      LODIndex=0
      PreviewMesh=&SM_Emote_SwissKale_Bat
      SoftSourceActor=null
      SourceMode=DefaultMeshOnly
      bAllowSamplingFromStreamingLODs=false
      bCaptureTransformsPerFrame=true
      bUsePhysicsBodyVelocity=false
      FilteredSockets []
      LODIndexUserParameter {parameter {name=None, typeDefHandle {registeredTypeIndex=108}, varData []}}
      MeshParameterBinding {parameter {name=None, typeDefHandle {registeredTypeIndex=1}, varData []}}
      SectionFilter {allowedMaterialSlots []}
     }}
   )
   /SetVariables(Particles.Position=&Multiply_VectorByFloat(
     Vector=$Particles.StaticMeshLocation.SampledPosition
     Float=&FN_LOD_Distance_Remap(Start Distance=300, End Distance=400, Minimum Output=0.8, Maximum Output=1.1, Clamp Output=true)
    )) {isSetParameters=true}
   &ParticleState()
   &ScaleSpriteSize(Uniform Curve Sprite Scale=DataInterface() {propertyValues {Curve {keys [{tangentMode=RCTM_User, time=0, value=1}, {tangentMode=RCTM_User, time=1, value=1.2}]}}})
   &SubUVAnimation(
    End Frame Range Override=40
    "Sprite Renderer "=DataInterface() {propertyValues {SpriteRenderer=&self:MeshSmokeBlur_2.NiagaraSpriteRendererProperties_0}}
    SubUV Lookup Index Scale=0.7
    UseEndFrame=true
    UseStartFrame=true
   )
   &ScaleColor(
    Scale Alpha=&FloatFromCurve(
     FloatCurve=DataInterface() {propertyValues {Curve {keys [
         {interpMode=RCIM_Cubic, tangentMode=RCTM_User, time=0, value=0}
         {interpMode=RCIM_Cubic, tangentMode=RCTM_User, time=0.2, value=1}
         {tangentMode=RCTM_User, time=1, value=0}
        ]}}}
     Scale Curve=&Multiply_Float(
      A=&Multiply_Float(
       A=&RandomRangeFloat(Minimum=0.7, Maximum=1)
       B=&FloatFromCurve(FloatCurve=DataInterface() {propertyValues {Curve {keys [<at 0 0>,<at 0.9 0>,<at 1 1>]}}}, CurveIndex=$Emitter.Age, Scale Curve=1.5)
      )
      B=&FN_LOD_Distance_Remap(Start Distance=300, End Distance=400, Minimum Output=0.6, Maximum Output=1, Clamp Output=true)
     )
    )
    Scale RGB=&Multiply_VectorByFloat(
     Vector=&MakeVectorFromLinearColorRGB(Color=&FNTimeOfDayColor(
       Night Color=<rgba 0.0667 0.0917 0.1667 1>
       Morning Color=<rgba 0.1667 0.1667 0.1667 1>
       Day Color=&MakeCustomLinearColorFromBool(Bool=$User.bIsFrontEndPreview, True Color=<rgba 0.0833 0.1042 0.1667 1>, False Color=<rgba 0.1667 0.1667 0.1667 1>)
       Evening Color=<rgba 0.1667 0.1667 0.1667 1>
      ))
     Float=&RandomRangeFloat(Minimum=0, Maximum=1)
    )
    ScaleA=true
    ScaleRGB=true
   )
   &DynamicMaterialParameters(
    Index 0 Param 1=&FloatFromCurve(
     FloatCurve=DataInterface() {propertyValues {Curve {keys [
         {interpMode=RCIM_Cubic, tangentMode=RCTM_User, time=0, value=0}
         <at 0.2 0.55>
         {arriveTangent=-1.0223, interpMode=RCIM_Cubic, leaveTangent=-1.0223, tangentMode=RCTM_User, time=1, value=1}
        ]}}}
     Scale Curve=&RandomRangeFloat(Minimum=3, Maximum=3)
    )
    Module.Index0Param0Write=true
    Module.Index0Param1Write=true
    Module.Index0Param2Write=true
    Module.Index0Param3Write=true
   )
   &GravityForce(Gravity=<0 0 100>) {enabled=false}
   &SolveForcesAndVelocity(Clamp Velocity=false, Limit Acceleration=false)
   /SetVariables(Particles.Position=$Particles.Position) {enabled=false, isSetParameters=true}
  ]
  rendererClasses [&NiagaraSpriteRendererProperties]
  renderers [{rendererClass=&NiagaraSpriteRendererProperties}]
 }
 Smokedetail {
  emitterUpdateScript [
   &EmitterState(Life Cycle Mode=::NewEnumerator1, Loop Behavior=::NewEnumerator1, Loop Delay=0.5, Loop Duration=2, UseLoopDelay=false)
   &SpawnBurst_Instantaneous(Use Spawn Probability=false, UseLoopCountLimit=false)
   &SpawnRate(
    Spawn Probability=&FN_LOD_Distance_Remap(Start Distance=300, End Distance=400, Minimum Output=1.5, Maximum Output=1, Clamp Output=true)
    SpawnRate=3
    Use Spawn Probability=true
   )
  ]
  particleSpawnScript [
   &PartitionParticles(Number Of Partitions=1)
   &InitializeParticle(
    Color=&MultiplyLinearColor(Linear Color=<rgba 0.7031 0.7031 0.7031 0.8>, Scale Factor=&MakeVector4_FromFloat(Float=1))
    Lifetime Max=1.6
    Lifetime Min=1.5
    Lifetime Mode=::NewEnumerator1
    Position Offset=<0 0 50>
    Sprite Rotation Mode=::NewEnumerator2
    Sprite Size Mode=::NewEnumerator1
    Uniform Sprite Size Max=150
    Uniform Sprite Size Min=130
    UsePositionOffset=true
    Write Material Random=true
   )
  ]
  particleUpdateScript [
   &StaticMeshLocation(
    MaskSpawnGroup=false
    Offset Position Along Sampled Normal=0.5
    OffsetAlongNormal=true
    OverrideRandomSeed=false
    SamplePosOffset=false
    Static Mesh=DataInterface() {propertyValues {
      DefaultMesh=&SM_Emote_SwissKale_Bat
      InstanceIndex=-1
      LODIndex=0
      PreviewMesh=&SM_Emote_SwissKale_Bat
      SoftSourceActor=null
      SourceMode=DefaultMeshOnly
      bAllowSamplingFromStreamingLODs=false
      bCaptureTransformsPerFrame=true
      bUsePhysicsBodyVelocity=false
      FilteredSockets []
      LODIndexUserParameter {parameter {name=None, typeDefHandle {registeredTypeIndex=108}, varData []}}
      MeshParameterBinding {parameter {name=None, typeDefHandle {registeredTypeIndex=1}, varData []}}
      SectionFilter {allowedMaterialSlots []}
     }}
   )
   /SetVariables(Particles.Position=&Multiply_VectorByFloat(
     Vector=$Particles.StaticMeshLocation.SampledPosition
     Float=&FN_LOD_Distance_Remap(Start Distance=300, End Distance=400, Minimum Output=0.9, Maximum Output=1, Clamp Output=true)
    )) {isSetParameters=true}
   &Drag(Drag=0, Use Linear Drag=true, Use Rotational Drag=true)
   &SolveForcesAndVelocity(Clamp Velocity=false, Limit Acceleration=false)
   &ParticleState()
   &SpriteRotationRate(Rotation Rate=&RandomRangeFloat(Minimum=-50, Maximum=50))
   &ScaleSpriteSize(
    Uniform Curve Scale=&FN_LOD_Distance_Remap(Start Distance=300, End Distance=400, Minimum Output=0.7, Maximum Output=1, Clamp Output=true)
    Uniform Curve Sprite Scale=DataInterface() {propertyValues {Curve {keys [
        {arriveTangent=8.7867, interpMode=RCIM_Cubic, leaveTangent=8.7867, tangentMode=RCTM_User, time=0, value=0}
        {arriveTangent=0.1614, interpMode=RCIM_Cubic, leaveTangent=0.1614, tangentMode=RCTM_User, time=0.3, value=0.9}
        <at 1 1>
       ]}}}
   )
   &DynamicMaterialParameters(
    Index 0 Param 1=&FloatFromCurve(
     FloatCurve=DataInterface() {propertyValues {Curve {keys [
         {interpMode=RCIM_Cubic, tangentMode=RCTM_User, time=0, value=0}
         <at 0.2 0.55>
         {arriveTangent=-1.0223, interpMode=RCIM_Cubic, leaveTangent=-1.0223, tangentMode=RCTM_User, time=1, value=1}
        ]}}}
     Scale Curve=&RandomRangeFloat(Minimum=3, Maximum=3)
    )
    Module.Index0Param0Write=true
    Module.Index0Param1Write=true
    Module.Index0Param2Write=true
    Module.Index0Param3Write=true
   )
   &ScaleColor(
    Scale Alpha=&FloatFromCurve(
     FloatCurve=DataInterface() {propertyValues {Curve {keys [
         {arriveTangent=3.6424, interpMode=RCIM_Cubic, leaveTangent=3.6424, tangentMode=RCTM_User, time=0, value=0}
         <at 0.5 1>
         {arriveTangent=-2, interpMode=RCIM_Cubic, leaveTangent=-2, time=0.7096, value=0.5753}
         <at 1 0>
        ]}}}
     Scale Curve=&Multiply_Float(
      A=1.5
      B=&FloatFromCurve(
       FloatCurve=DataInterface() {propertyValues {Curve {keys [<at 0 0>,<at 0.9 0>,<at 1 1>]}}}
       CurveIndex=$Emitter.Age
       Scale Curve=&FN_FrontEndFloatModifier(Gameplay=1, Front End=0.5)
      )
     )
    )
    Scale RGB=&MakeCustomVectorFromBool(
     Bool=$User.bIsFrontEndPreview
     True Vector=&MakeVectorFromLinearColorRGB(Color=&ColorFromCurve(ColorCurve=DataInterface() {propertyValues {
         AlphaCurve {keys []}
         BlueCurve {keys [{arriveTangent=-331.6383, interpMode=RCIM_Cubic, time=0, value=0.07}, <at 0.7598 0 RCIM_Cubic>]}
         GreenCurve {keys [{arriveTangent=-87.3695, interpMode=RCIM_Cubic, time=0, value=0.0546}, <at 0.7598 0 RCIM_Cubic>]}
         RedCurve {keys [<at 0 0.042 RCIM_Cubic>,<at 0.7598 0 RCIM_Cubic>]}
        }}))
     False Vector=&MakeVectorFromLinearColorRGB(Color=&FNTimeOfDayColor(
       Night Color=&ColorFromCurve(ColorCurve=DataInterface() {propertyValues {
          AlphaCurve {keys []}
          BlueCurve {keys [{arriveTangent=-331.6383, interpMode=RCIM_Cubic, time=0, value=0.1}, <at 0.7598 0.05 RCIM_Cubic>]}
          GreenCurve {keys [{arriveTangent=-87.3695, interpMode=RCIM_Cubic, time=0, value=0.055}, <at 0.7598 0.0312 RCIM_Cubic>]}
          RedCurve {keys [<at 0 0.04 RCIM_Cubic>,<at 0.7598 0.025 RCIM_Cubic>]}
         }})
       Morning Color=&ColorFromCurve(ColorCurve=DataInterface() {propertyValues {
          AlphaCurve {keys []}
          BlueCurve {keys [{arriveTangent=-331.6383, interpMode=RCIM_Cubic, time=0, value=0.0781}, <at 0.7598 0 RCIM_Cubic>]}
          GreenCurve {keys [{arriveTangent=-87.3695, interpMode=RCIM_Cubic, time=0, value=0.0781}, <at 0.7598 0 RCIM_Cubic>]}
          RedCurve {keys [<at 0 0.0781 RCIM_Cubic>,<at 0.7598 0 RCIM_Cubic>]}
         }})
       Day Color=&ColorFromCurve(ColorCurve=DataInterface() {propertyValues {
          AlphaCurve {keys []}
          BlueCurve {keys [{arriveTangent=-331.6383, interpMode=RCIM_Cubic, time=0, value=0.0781}, <at 0.7598 0 RCIM_Cubic>]}
          GreenCurve {keys [{arriveTangent=-87.3695, interpMode=RCIM_Cubic, time=0, value=0.0781}, <at 0.7598 0 RCIM_Cubic>]}
          RedCurve {keys [<at 0 0.0781 RCIM_Cubic>,<at 0.7598 0 RCIM_Cubic>]}
         }})
       Evening Color=&ColorFromCurve(ColorCurve=DataInterface() {propertyValues {
          AlphaCurve {keys []}
          BlueCurve {keys [{arriveTangent=-331.6383, interpMode=RCIM_Cubic, time=0, value=0.0781}, <at 0.7598 0 RCIM_Cubic>]}
          GreenCurve {keys [{arriveTangent=-87.3695, interpMode=RCIM_Cubic, time=0, value=0.0781}, <at 0.7598 0 RCIM_Cubic>]}
          RedCurve {keys [<at 0 0.0781 RCIM_Cubic>,<at 0.7598 0 RCIM_Cubic>]}
         }})
      ))
    )
    ScaleA=true
    ScaleRGB=true
   )
   &SubUVAnimation(
    End Frame Range Override=30
    "Sprite Renderer "=DataInterface() {propertyValues {SpriteRenderer=&self:Smokedetail_3.NiagaraSpriteRendererProperties_0}}
    Start Frame Range Override=&RandomRangeInt(Minimum=0, Maximum=5)
    UseEndFrame=true
    UseStartFrame=true
   )
  ]
  rendererClasses [&NiagaraSpriteRendererProperties]
  renderers [{rendererClass=&NiagaraSpriteRendererProperties}]
 }
}
summary {
 systemName=NS_Emote_SwissKale_Bat
 emitters [
  {emitterName=BatMesh, rendererClasses [&NiagaraMeshRendererProperties]}
  {emitterName=BatMesh02, rendererClasses [&NiagaraMeshRendererProperties]}
  {emitterName=MeshSmokeBlur, rendererClasses [&NiagaraSpriteRendererProperties]}
  {emitterName=Smokedetail, rendererClasses [&NiagaraSpriteRendererProperties]}
 ]
 userVariables [
  User.bIsFrontEnd {type=NiagaraBool}
  User.bIsFrontEndPreview {type=NiagaraBool}
 ]
}
