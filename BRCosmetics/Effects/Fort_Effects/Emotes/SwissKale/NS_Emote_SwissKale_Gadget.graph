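# Niagara system for the SwissKale emote's gadget, with a single emitter named Ball that renders a mesh particle.
# An instantaneous burst spawns the particle with a three-second lifetime and a material random value.
# The update stage runs only particle state, with no motion or force modules.
# User booleans for front-end and front-end preview are exposed, but none of the shown modules read them.
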NiagaraSystem /BRCosmetics/Effects/Fort_Effects/Emotes/SwissKale/NS_Emote_SwissKale_Gadget
dependencies {usedDataInterfaces [], usedDynamicInputs [], usedModules [&SystemState]}
emitters {Ball {
  emitterUpdateScript [
   &EmitterState()
   &SpawnBurst_Instantaneous(Use Spawn Probability=false, UseLoopCountLimit=false)
  ]
  particleSpawnScript [&InitializeParticle(
    Color Mode=::NewEnumerator0
    Lifetime=3
    Mesh Scale Mode=::NewEnumerator3
    Sprite Size Mode=::NewEnumerator3
    Uniform Sprite Size=100
    UsePositionOffset=false
    Write Material Random=true
   )]
  particleUpdateScript [&ParticleState()]
  rendererClasses [&NiagaraMeshRendererProperties]
  renderers [{rendererClass=&NiagaraMeshRendererProperties, rendererIndex=0}]
 }}
summary {
 systemName=NS_Emote_SwissKale_Gadget
 emitters [{emitterName=Ball, rendererClasses [&NiagaraMeshRendererProperties]}]
 userVariables [
  User.bIsFrontEnd {type=NiagaraBool}
  User.bIsFrontEndPreview {type=NiagaraBool}
 ]
}
